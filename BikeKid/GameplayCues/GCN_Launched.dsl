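; Gameplay cue notify for the BikeKid launched event, handling the visual switch on the target BP_BikeKid when executed.
; OnExecute casts the target to BP_BikeKid, stores it locally, and spawns a rocket transform particle effect at its location and rotation.
; It then shows the character mesh and blinking light particles, hides the LoadMissile component, and returns the parent implementation's result.
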(fn OnExecute
  (params (MyTarget :type Actor%20Object%20Reference) (Parameters :type Gameplay%20Cue%20Parameters%20Structure%20%28by%20ref%29))
  (bind _bikekid (LocalVariables|GetBikeKid))
  (bind _asbp_bike_kid (Utilities|Casting|CastToBP_BikeKid MyTarget))
  (bind _output_get (LocalVariables|SetBikeKid _asbp_bike_kid))
  (Effects|SpawnEmitteratLocation 0 "/BikeKid/FX/P_RC_Rocket_Transform.P_RC_Rocket_Transform" (Transformation|GetActorLocation _output_get) (Transformation|GetActorRotation _output_get))
  (Rendering|SetVisibility (Variables|/Script/Engine.Character|GetMesh _bikekid) true false)
  (Rendering|SetVisibility (Variables|/BikeKid/Items/Blueprints/BP_BikeKid.BP_BikeKid_C|GetLoadMissile _bikekid) false false)
  (Rendering|SetVisibility (Variables|/BikeKid/Items/Blueprints/BP_BikeKid.BP_BikeKid_C|GetPS_BlinkingLight _bikekid) true false)
  (return (|CallParentFunction MyTarget Parameters)))
